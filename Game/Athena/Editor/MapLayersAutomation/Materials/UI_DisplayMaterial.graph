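# UI-domain material in the Athena editor's MapLayersAutomation folder, apparently for displaying a map capture image.
# Its only node is a 2D texture parameter named Image, defaulting to the Capture_Iteration_Discovered texture.
# The graph shows no other processing and no link from that sample to a material output.
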Material /Game/Athena/Editor/MapLayersAutomation/Materials/UI_DisplayMaterial
graph {
 properties {materialDomain=MD_UI, pixelDepthOffsetMode=PDOM_Legacy, usedShadingModels=Default Lit}
 nodes [TextureSampleParameter2D() {parameterName=Image, texture=/Game/Athena/Editor/Capture_Iteration_Discovered}]
}
