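# EUW_ResizeDivider is an editor utility drag handle that resizes the two sibling widgets either side of it in a Horizontal or Vertical Box.
# Pressing its button captures both slot sizes and opens a gate, so Tick runs ResizePanels until release.
# ResizePanels maps mouse position across both neighbors to redistribute their combined size value, clamped to a minimum, when size rules match FillConst.
# An invalid parent flags the divider for removal.
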EditorUtilityWidgetBlueprint /Game/Athena/Editor/MapLayersAutomation/SubWidgets/EUW_ResizeDivider
graphs [
 DestroyIfMissingBeforeAfter {nodes [
   DestroyIfMissingBeforeAfter() {is=title K2Node_FunctionEntry}
   > n1 = Branch(Condition=Get Before After Widgets!() -> (
     Before:Widget Object Reference
     BeforeSize:Slate Child Size Structure
     After:Widget Object Reference
     AfterSize:Slate Child Size Structure
     IsValid:bool=~
    ) {is=title K2Node_MacroInstance}.IsValid)
   DestroySelf(execute=@n1.else)
  ]}
 EventGraph {nodes [
   "On Pressed (Button_0)"() {is=title K2Node_ComponentBoundEvent}
   > n1 = GetCurrentSizes()
   n2 = Gate!(Enter:Exec=@n3.then, Open:Exec=@n1.then, Close:Exec=@n4.then, Toggle:Exec, bStartClosed:bool=true) -> (Exit:Exec) {is=title K2Node_MacroInstance}
   n3 = Event Tick()
   Set My Geometry(execute=@n2.Exit, My Geometry:Geometry Structure=@n3.MyGeometry) -> (Output_Get:Geometry Structure)
   > Resize()
   n4 = "On Released (Button_0)"() {is=title K2Node_ComponentBoundEvent}
   DestroySelf() {is=title K2Node_CustomEvent}
   > Set QueueDestroy(QueueDestroy:bool=true) -> (Output_Get:bool)
   Resize() {is=title K2Node_CustomEvent}
   > ResizePanels()
   Vertical() {is=title K2Node_CustomEvent}
   > DestroySelf() {is=title K2Node_CallFunction}
   Destroy() {is=title K2Node_CustomEvent}
   > Branch(Condition=Get QueueDestroy() -> (QueueDestroy:bool))
   > RemoveFromParent()
   Event Construct()
   > Set Cursor!(
    execute
    Cursor:EMouseCursor Enum=Select(Option 0:EMouseCursor Enum=ResizeUpDown, Option 1:EMouseCursor Enum=ResizeLeftRight, Index=Get bVertical() -> (bVertical:bool)) -> (ReturnValue:EMouseCursor Enum)
    self:Widget Object Reference=Get Button_0() -> (Button_0:Button Object Reference)
   ) -> (then, Output_Get:EMouseCursor Enum=None)
   > Set Timer by Event(Delegate=Create Event(), Time=0.05, bLooping=false)
   > n5 = Get!(execute:Exec) -> (then:Exec, else:Exec, bVertical:bool)
   n6 = GetParent!(self:Widget Object Reference) -> (ReturnValue:Panel Widget Object Reference)
   > n7 = Cast To HorizontalBox!(execute, Object=@n6) -> (then, CastFailed, AsHorizontal Box:Horizontal Box Object Reference) {is=title K2Node_DynamicCast}
   > Set As Horizontal Box(As Horizontal Box:Horizontal Box Object Reference=@n7) -> (Output_Get:Horizontal Box Object Reference)
   PrintString(execute=[@n7.CastFailed, @n8.CastFailed], InString="Invalid Parent. Autodestruct!")
   > Vertical() {is=title K2Node_CallFunction}
   n8 = Cast To VerticalBox!(execute=@n5.else, Object=@n6) -> (then, CastFailed, AsVertical Box:Vertical Box Object Reference) {is=title K2Node_DynamicCast}
   > Set AsVerticalBox(AsVerticalBox:Vertical Box Object Reference=@n8) -> (Output_Get:Vertical Box Object Reference)
  ]}
 GetBeforeAfterWidgets {nodes [
   Inputs!()
   n1 = GetParent()
   n2 = GetChildIndex!(self:Panel Widget Object Reference=@n1, Content:Widget Object Reference=Self-Reference()) -> (ReturnValue:int=0)
   n3 = GetChildAt!(self:Panel Widget Object Reference=@n1, Index:int=int - int(A=@n2)) -> (ReturnValue:Widget Object Reference)
   n4 = GetChildAt(self=@n1, Index=int + int(A=@n2))
   n5 = SlotAsHorizontalBoxSlot!(Widget:Widget Object Reference=@n3) -> (ReturnValue:Horizontal Box Slot Object Reference)
   n6 = IsValid(Object=@n5)
   n7 = SlotAsHorizontalBoxSlot(Widget=@n4)
   n8 = IsValid(Object=@n7)
   n9 = Get bVertical()
   n10 = NOT Boolean(A=@n9)
   n11 = SlotAsVerticalBoxSlot(Widget=@n4)
   n12 = IsValid(Object=@n11)
   n13 = SlotAsVerticalBoxSlot(Widget=@n3)
   n14 = IsValid(Object=@n13)
   Outputs!(
    Before:Widget Object Reference=@n3
    BeforeSize:Slate Child Size Structure=Select(
     Option 0:Slate Child Size Structure=Get Size!(self:Vertical Box Slot Object Reference=@n13) -> (Size:Slate Child Size Structure)
     Option 1:Slate Child Size Structure=Get Size(self:Horizontal Box Slot Object Reference=@n5)
     Index=@n9
    ) -> (ReturnValue:Slate Child Size Structure)
    After:Widget Object Reference=@n4
    AfterSize:Slate Child Size Structure=Select(Option 0=Get Size(self:Vertical Box Slot Object Reference=@n11), Option 1=Get Size(self:Horizontal Box Slot Object Reference=@n7), Index=@n9)
    IsValid:bool=OR Boolean(A=AND Boolean!(A=@n14, B=@n12, C:bool=@n10) -> (ReturnValue), B=AND Boolean(A=@n6, B=@n8, C=@n9))
   )
  ]}
 GetCurrentSizes {nodes [
   GetCurrentSizes() {is=title K2Node_FunctionEntry}
   n1 = Get Before After Widgets() -> (IsValid=~)
   > Set BeforeSize!(execute, BeforeSize:Slate Child Size Structure=@n1.BeforeSize) -> (then, Output_Get_Value:float=~, Output_Get_SizeRule:ESlateSizeRule Enum)
   > Set AfterSize(AfterSize:Slate Child Size Structure=@n1.AfterSize) -> (Output_Get:Slate Child Size Structure)
  ]}
 ResizePanels {nodes [
   ResizePanels() {is=title K2Node_FunctionEntry}
   > n1 = GetMousePositionOnViewport!(execute, WorldContextObject:Object Reference) -> (then, ReturnValue:Vector 2D Structure)
   > Set MousePosition(MousePosition:Vector 2D Structure=@n1) -> (Output_Get:Vector 2D Structure)
   n2 = Get FillConst() -> (FillConst:ESlateSizeRule Enum=Automatic)
   n3 = Get!() -> (BeforeSize_Value:float=~, BeforeSize_SizeRule:ESlateSizeRule Enum)
   n4 = Get!() -> (AfterSize_Value:float=~, AfterSize_SizeRule:ESlateSizeRule Enum)
   n5 = Get Before After Widgets() -> (IsValid=~)
   > Branch(Condition=AND Boolean(
     A=@n5.IsValid
     B="Equal (Enum)"(A:ESlateSizeRule Enum=@n3.BeforeSize_SizeRule, B:ESlateSizeRule Enum=@n2)
     C="Equal (Enum)"(A=@n4.AfterSize_SizeRule, B=@n2)
    ))
   > n6 = Get!(execute:Exec) -> (then:Exec, else:Exec, bVertical:bool)
   n7 = GetAbsolutePosition!(Geometry:"Geometry Structure (by ref)"=GetTickSpaceGeometry!(self:Widget Object Reference=@n5) -> (ReturnValue:"Geometry Structure (by ref)")) -> (ReturnValue_X:double, ReturnValue_Y:double)
   n8 = Get bVertical()
   n9 = Get!() -> (MousePosition_X:double, MousePosition_Y:double)
   n10 = float + float(A=@n3, B=@n4)
   n11 = GetTickSpaceGeometry(self=@n5.After)
   n12 = GetAbsoluteSize!(Geometry:"Geometry Structure (by ref)"=@n11) -> (ReturnValue_X:double, ReturnValue_Y:double)
   n13 = SelectFloat(A=@n12, B=@n12.ReturnValue_Y, bPickA=@n8)
   n14 = GetAbsolutePosition(Geometry=@n11)
   n15 = SelectFloat(A=@n14, B=@n14.ReturnValue_Y, bPickA=@n8)
   n16 = "Clamp (Float)"(
    Value=MapRangeClamped(
     Value=SelectFloat(A=@n9, B=@n9.MousePosition_Y, bPickA=@n8)
     InRangeA=SelectFloat(A=@n7, B=@n7.ReturnValue_Y, bPickA=@n8)
     InRangeB=float + float(A=@n15, B=@n13)
     OutRangeB=@n10
    )
    Min=0.05
    Max=float - float(A=@n10, B=0.05)
   )
   > SetHorizontalBoxSize!(execute, self, InSize_Value:float=@n16, Widget:Widget Object Reference=@n5, InSize_SizeRule:ESlateSizeRule Enum=@n3.BeforeSize_SizeRule) -> (then)
   n17 = float - float(A=@n10, B=@n16)
   > SetHorizontalBoxSize(InSize_Value=@n17, Widget=@n5.After, InSize_SizeRule=@n4.AfterSize_SizeRule)
   SetVerticalBoxSize!(
    execute=@n6.else
    self
    InSize_Value:float=@n16
    Widget:Widget Object Reference=@n5
    InSize_SizeRule:ESlateSizeRule Enum=@n3.BeforeSize_SizeRule
   ) -> (then)
   > SetVerticalBoxSize(InSize_Value=@n17, Widget=@n5.After, InSize_SizeRule=@n4.AfterSize_SizeRule)
  ]}
 SetHorizontalBoxSize {nodes [
   n1 = SetHorizontalBoxSize!() -> (then, InSize_Value:float=~, Widget:Widget Object Reference, InSize_SizeRule:ESlateSizeRule Enum=Automatic) {is=title K2Node_FunctionEntry}
   > SetSize!(
    execute
    self:Horizontal Box Slot Object Reference=SlotAsHorizontalBoxSlot(Widget=@n1.Widget)
    InSize_Value:float=@n1
    InSize_SizeRule:ESlateSizeRule Enum=@n1.InSize_SizeRule
   ) -> (then)
  ]}
 SetVerticalBoxSize {nodes [
   n1 = SetVerticalBoxSize!() -> (then, InSize_Value:float=~, Widget:Widget Object Reference, InSize_SizeRule:ESlateSizeRule Enum=Automatic) {is=title K2Node_FunctionEntry}
   > SetSize(self:Vertical Box Slot Object Reference=SlotAsVerticalBoxSlot(Widget=@n1.Widget), InSize_Value=@n1, InSize_SizeRule=@n1.InSize_SizeRule)
  ]}
]
variables [
 AfterSize {type=SlateChildSize, default {sizeRule=Fill, value=1}}
 As Horizontal Box {type=HorizontalBox, default=null}
 AsVerticalBox {type=VerticalBox, default=null}
 BeforeSize {type=SlateChildSize, default {sizeRule=Fill, value=1}}
 bVertical {type=boolean, default=true}
 My Geometry {type=Geometry, default {}}
 QueueDestroy {type=boolean}
 Timer {type=TimerHandle}
]
widgets {
 info {rootWidgetClass=&Button, widgetCount=2}
 widgets [
  Button_0 {
   widgetClassPath=&Button
   widget=Button_0
   properties {
    bOverride_Cursor=true
    cursor=ResizeLeftRight
    widgetStyle {
     clickedSlateSound {resourceObject=null}
     disabled {
      bIsDynamicallyLoaded=false
      drawAs=Box
      imageType=NoImage
      mirroring=NoMirror
      resourceName=None
      resourceObject=null
      tiling=NoTile
      imageSize=<0 0>
      margin=<ltrb 0 0 0 0>
      outlineSettings {
       bUseBrushTransparency=false
       roundingType=FixedRadius
       width=1
       color {colorUseRule=UseColor_ColorTable, specifiedColor=<rgba 0 0 0 0>}
       cornerRadii=<4 4 4 4>
      }
      tintColor {colorUseRule=UseColor_Specified, specifiedColor=<rgba 0.0395 0.0395 0.0395 1>}
      uVRegion {bIsValid=false, max=<0 0>, min=<0 0>}
     }
     disabledForeground {colorUseRule=UseColor_ColorTable, specifiedColor=<rgba 0 0 0 0>}
     hovered {
      bIsDynamicallyLoaded=false
      drawAs=Box
      imageType=NoImage
      mirroring=NoMirror
      resourceName=None
      resourceObject=null
      tiling=NoTile
      imageSize=<5 5>
      margin=<ltrb 0 0 0 0>
      outlineSettings {
       bUseBrushTransparency=false
       roundingType=FixedRadius
       width=1
       color {colorUseRule=UseColor_ColorTable, specifiedColor=<rgba 0 0 0 0>}
       cornerRadii=<4 4 4 4>
      }
      tintColor {colorUseRule=UseColor_Specified, specifiedColor=<rgba 0.032 0.032 0.032 1>}
      uVRegion {bIsValid=false, max=<0 0>, min=<0 0>}
     }
     hoveredForeground {colorUseRule=UseColor_ColorTable, specifiedColor=<rgba 0 0 0 0>}
     hoveredSlateSound {resourceObject=null}
     normal {
      bIsDynamicallyLoaded=false
      drawAs=Image
      imageType=FullColor
      mirroring=NoMirror
      resourceName=None
      resourceObject=null
      tiling=NoTile
      imageSize=<128 128>
      margin=<ltrb 0 0 0 0>
      outlineSettings {
       bUseBrushTransparency=false
       roundingType=FixedRadius
       width=1
       color {colorUseRule=UseColor_ColorTable, specifiedColor=<rgba 0 0 0 0>}
       cornerRadii=<4 4 4 4>
      }
      tintColor {colorUseRule=UseColor_Specified, specifiedColor=<rgba 0.0075 0.0075 0.0075 1>}
      uVRegion {bIsValid=false, max=<0 0>, min=<0 0>}
     }
     normalForeground {colorUseRule=UseColor_ColorTable, specifiedColor=<rgba 0 0 0 0>}
     normalPadding=<ltrb 0 0 0 0>
     pressed {
      bIsDynamicallyLoaded=false
      drawAs=Box
      imageType=NoImage
      mirroring=NoMirror
      resourceName=None
      resourceObject=null
      tiling=NoTile
      imageSize=<5 5>
      margin=<ltrb 0 0 0 0>
      outlineSettings {
       bUseBrushTransparency=false
       roundingType=FixedRadius
       width=1
       color {colorUseRule=UseColor_ColorTable, specifiedColor=<rgba 0 0 0 0>}
       cornerRadii=<4 4 4 4>
      }
      tintColor {colorUseRule=UseColor_Specified, specifiedColor=<rgba 0.032 0.032 0.032 1>}
      uVRegion {bIsValid=false, max=<0 0>, min=<0 0>}
     }
     pressedForeground {colorUseRule=UseColor_ColorTable, specifiedColor=<rgba 0 0 0 0>}
     pressedPadding=<ltrb 0 0 0 0>
     pressedSlateSound {resourceObject=null}
    }
   }
  }
  Image_34 {
   widgetClassPath=&Image
   bIsVariable=true
   parent=Button_0
   widget=Image_34
   properties {brush {
     bIsDynamicallyLoaded=false
     drawAs=Image
     imageType=FullColor
     mirroring=NoMirror
     resourceName=None
     resourceObject=&/T_ResizeDivider_Handle
     tiling=NoTile
     imageSize=<24 7>
     margin=<ltrb 0 0 0 0>
     outlineSettings {
      bUseBrushTransparency=false
      roundingType=HalfHeightRadius
      width=0
      color {colorUseRule=UseColor_Specified, specifiedColor=<rgba 0 0 0 0>}
      cornerRadii=<0 0 0 0>
     }
     tintColor {colorUseRule=UseColor_Specified, specifiedColor=<rgba 1 1 1 0.35>}
     uVRegion {bIsValid=false, max=<0 0>, min=<0 0>}
    }}
   slot {class=ButtonSlot, properties {padding=<ltrb 0 0 0 0>}}
  }
 ]
}
